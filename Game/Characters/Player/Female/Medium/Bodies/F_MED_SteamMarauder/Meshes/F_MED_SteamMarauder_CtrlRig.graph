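# Control Rig for the F_MED_SteamMarauder player body that adds procedural corrective motion to body deform bones, knee armor and skirt bones.
# A shared DeformRig function runs first, then a sequence parent-constrains the right knee armor between thigh and calf, nudging it with a calf pose reader.
# Spherical pose readers on both thighs store up, side, back and crossed-step weights in variables.
# These weights drive additive skirt bone offsets and reduce the back skirt rags' pelvis/thigh constraint as legs move back.
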ControlRigBlueprint /Game/Characters/Player/Female/Medium/Bodies/F_MED_SteamMarauder/Meshes/F_MED_SteamMarauder_CtrlRig
graphs [
 RigVMFunctionLibrary {links [], nodes []}
 RigVMFunction_Sequence_1_ContainedGraph {nodes [
   n1 = RigVMFunctionEntryNode()
   n2 = &RigVMFunction_Sequence(ExecuteContext=@n1.ExecuteContext)
   n3 = &RigVMFunction_Sequence(ExecuteContext=@n2.B)
   n4 = &RigVMFunction_Sequence(ExecuteContext=@n3.B)
   n5 = &RigVMFunction_Sequence(ExecuteContext=@n4.B)
   RigVMFunctionReturnNode(B=@n2.A, D=@n4.A, E=@n5.A, F=@n5.B, C=@n3.A)
  ]}
 RigVMModel {nodes [
   RigUnit_BeginExecution = &RigUnit_BeginExecution()
   FN_DeformRig_FNC = RigVMFunctionReferenceNode(
    DeformRig_ Curve="(deformRig_l_upperarm_up_alpha_72_#1=False,deformRig_r_upperarm_up_alpha_73_#2=False,deformRig_l_upperarm_fr_alpha_74_#3=False,deformRig_r_upperarm_fr_alpha_75_CB51B8 ...(truncated)"
    DeformRig_Bool="(upperarm_1_#4=True,elbow_3_#5=True,wrist_7_#6=True,groin_8_#7=True,glute_17_#8=true,knee_9_40C408324CA261 ...(truncated)"
    DeformRig_Debug=(upperarm_debug_18_#4=False,wrist_debug_20_#6=False,lat_debug_26_#9=False,groinGlute_debug_27_#7=False)
    DeformRig_Value="(upperarm_up_18_#10=(Rotation=(X=0,Y=0,Z=0,W=1),Translation=(X=0,Y=8,Z=8),Scale3D=(X=0.7,Y=1,Z=1)),upperarm_fwd_fr_19_#11=(Rotation=(X=0,Y=0,Z=0,W=1),Translation=(X=0,Y=10,Z=0),Scale3D=(X=0.5,Y=1,Z= ...(truncated)"
    ExecuteContext=@RigUnit_BeginExecution.ExecutePin
   )
   > RigVMFunction_Sequence = RigVMAggregateNode()
   &RigUnit_ParentConstraint(
    AdvancedSettings=(InterpolationType=Average,RotationOrderForFilter=XZY)
    Child=(Type=Bone,Name="kneeArmor_r")
    Parents=((Item=(Type=Bone,Name="thigh_r"),Weight=1),(Item=(Type=Bone,Name="calf_r"),Weight=1))
    Weight=1
    ExecutePin=@RigVMFunction_Sequence.B
   )
   > n1 = &RigUnit_SphericalPoseReader(
    ActiveRegionScaleFactors=(PositiveWidth=1,NegativeWidth=1,PositiveHeight=1,NegativeHeight=1)
    ActiveRegionSize=0.1
    Debug=(bDrawDebug=False,bDraw2D=False,bDrawLocalAxes=False,DebugScale=12,DebugSegments=20,DebugThickness=0.25)
    DriverAxis=(X=1,Y=0,Z=0)
    DriverItem=(Type=Bone,Name="calf_r")
    FalloffRegionScaleFactors=(PositiveWidth=1,NegativeWidth=1,PositiveHeight=1,NegativeHeight=1)
    FalloffSize=0.3
    FlipHeightScaling=False
    FlipWidthScaling=False
    OptionalParentItem=(Type=Bone,Name="None")
    RotationOffset=(X=90,Y=0,Z=15)
   )
   > &RigUnit_ModifyTransforms(
    ItemToModify=((Item=(Type=Bone,Name="kneeArmor_r"),Transform=(Rotation=(X=0,Y=0,Z=0,W=1),Translation=(X=0,Y=-1.5992,Z=0.295),Scale3D=(X=1,Y=1,Z=1))))
    WeightMinimum=0
    Weight=@n1.OutputParam
   )
   n2 = &RigUnit_SphericalPoseReader(
    ActiveRegionScaleFactors=(PositiveWidth=1,NegativeWidth=1,PositiveHeight=1,NegativeHeight=1)
    ActiveRegionSize=0.07
    Debug=(bDrawDebug=false,bDraw2D=False,bDrawLocalAxes=False,DebugScale=25,DebugSegments=20,DebugThickness=0.1)
    DriverAxis=(X=-1,Y=0,Z=0)
    DriverItem=(Type=Bone,Name="thigh_l")
    FalloffRegionScaleFactors=(PositiveWidth=1,NegativeWidth=1,PositiveHeight=1,NegativeHeight=1)
    FalloffSize=0.3
    FlipHeightScaling=False
    FlipWidthScaling=False
    OptionalParentItem=(Type=Bone,Name="None")
    RotationOffset=(X=100.6701,Y=0,Z=-24.5322)
    ExecutePin=@RigVMFunction_Sequence.C
   )
   > RigVMVariableNode(Value=@n2.OutputParam) {variable=thigh_l_up}
   > n3 = &RigUnit_SphericalPoseReader(
    ActiveRegionScaleFactors=(PositiveWidth=1,NegativeWidth=1,PositiveHeight=1,NegativeHeight=1)
    ActiveRegionSize=0.15
    Debug=(bDrawDebug=False,bDraw2D=False,bDrawLocalAxes=False,DebugScale=25,DebugSegments=20,DebugThickness=0.25)
    DriverAxis=(X=-1,Y=0,Z=0)
    DriverItem=(Type=Bone,Name="thigh_l")
    FalloffRegionScaleFactors=(PositiveWidth=3,NegativeWidth=3,PositiveHeight=3,NegativeHeight=3)
    FalloffSize=0.2
    FlipHeightScaling=False
    FlipWidthScaling=False
    OptionalParentItem=(Type=Bone,Name="None")
    RotationOffset=(X=180,Y=-45,Z=0)
   )
   > RigVMVariableNode(Value=@n3.OutputParam) {variable=thigh_l_side}
   > n4 = &RigUnit_SphericalPoseReader(
    ActiveRegionScaleFactors=(PositiveWidth=1,NegativeWidth=1,PositiveHeight=1,NegativeHeight=1)
    ActiveRegionSize=0.05
    Debug=(bDrawDebug=False,bDraw2D=False,bDrawLocalAxes=False,DebugScale=25,DebugSegments=20,DebugThickness=0.25)
    DriverAxis=(X=-1,Y=0,Z=0)
    DriverItem=(Type=Bone,Name="thigh_l")
    FalloffRegionScaleFactors=(PositiveWidth=1,NegativeWidth=1,PositiveHeight=1,NegativeHeight=1)
    FalloffSize=0.25
    FlipHeightScaling=False
    FlipWidthScaling=False
    OptionalParentItem=(Type=Bone,Name="None")
    RotationOffset=(X=-35,Y=90,Z=0)
   )
   > RigVMVariableNode(Value=@n4.OutputParam) {variable=thigh_l_back}
   > n5 = &RigUnit_SphericalPoseReader(
    ActiveRegionScaleFactors=(PositiveWidth=1,NegativeWidth=1,PositiveHeight=1,NegativeHeight=1)
    ActiveRegionSize=0.15
    Debug=(bDrawDebug=false,bDraw2D=false,bDrawLocalAxes=true,DebugScale=25,DebugSegments=20,DebugThickness=0.25)
    DriverAxis=(X=1,Y=0,Z=0)
    DriverItem=(Type=Bone,Name="thigh_r")
    FalloffRegionScaleFactors=(PositiveWidth=1,NegativeWidth=1,PositiveHeight=1,NegativeHeight=1)
    FalloffSize=0.3
    FlipHeightScaling=False
    FlipWidthScaling=False
    OptionalParentItem=(Type=Bone,Name="thigh_l")
    RotationOffset=(X=-17.024,Y=-110,Z=76.3261)
   )
   > RigVMVariableNode(Value=@n5.OutputParam) {variable=Crossed_step_l}
   > n6 = &RigUnit_SphericalPoseReader(
    ActiveRegionScaleFactors=(PositiveWidth=1,NegativeWidth=1,PositiveHeight=1,NegativeHeight=1)
    ActiveRegionSize=0.05
    Debug=(bDrawDebug=False,bDraw2D=False,bDrawLocalAxes=False,DebugScale=25,DebugSegments=20,DebugThickness=0.15)
    DriverAxis=(X=1,Y=0,Z=0)
    DriverItem=(Type=Bone,Name="thigh_r")
    FalloffRegionScaleFactors=(PositiveWidth=1,NegativeWidth=1,PositiveHeight=1,NegativeHeight=1)
    FalloffSize=0.3
    FlipHeightScaling=False
    FlipWidthScaling=False
    OptionalParentItem=(Type=Bone,Name="None")
    RotationOffset=(X=-73.9712,Y=-0.373,Z=-26.9578)
   )
   > RigVMVariableNode(Value=@n6.OutputParam) {variable=thigh_r_up}
   > n7 = &RigUnit_SphericalPoseReader(
    ActiveRegionScaleFactors=(PositiveWidth=1,NegativeWidth=1,PositiveHeight=1,NegativeHeight=1)
    ActiveRegionSize=0.15
    Debug=(bDrawDebug=False,bDraw2D=false,bDrawLocalAxes=False,DebugScale=25,DebugSegments=20,DebugThickness=0.25)
    DriverAxis=(X=1,Y=0,Z=0)
    DriverItem=(Type=Bone,Name="thigh_r")
    FalloffRegionScaleFactors=(PositiveWidth=3,NegativeWidth=3,PositiveHeight=3,NegativeHeight=3)
    FalloffSize=0.2
    FlipHeightScaling=False
    FlipWidthScaling=False
    OptionalParentItem=(Type=Bone,Name="None")
    RotationOffset=(X=0,Y=-45,Z=0)
   )
   > RigVMVariableNode(Value=@n7.OutputParam) {variable=thigh_r_side}
   > n8 = &RigUnit_SphericalPoseReader(
    ActiveRegionScaleFactors=(PositiveWidth=1,NegativeWidth=1,PositiveHeight=1,NegativeHeight=1)
    ActiveRegionSize=0.05
    Debug=(bDrawDebug=False,bDraw2D=False,bDrawLocalAxes=False,DebugScale=25,DebugSegments=20,DebugThickness=0.25)
    DriverAxis=(X=1,Y=0,Z=0)
    DriverItem=(Type=Bone,Name="thigh_r")
    FalloffRegionScaleFactors=(PositiveWidth=1,NegativeWidth=1,PositiveHeight=1,NegativeHeight=1)
    FalloffSize=0.25
    FlipHeightScaling=False
    FlipWidthScaling=False
    OptionalParentItem=(Type=Bone,Name="None")
    RotationOffset=(X=-215,Y=90,Z=0)
   )
   > RigVMVariableNode(Value=@n8.OutputParam) {variable=thigh_r_back}
   > n9 = &RigUnit_SphericalPoseReader(
    ActiveRegionScaleFactors=(PositiveWidth=1,NegativeWidth=1,PositiveHeight=1,NegativeHeight=1)
    ActiveRegionSize=0.1
    Debug=(bDrawDebug=false,bDraw2D=false,bDrawLocalAxes=true,DebugScale=25,DebugSegments=20,DebugThickness=0.25)
    DriverAxis=(X=-1,Y=0,Z=0)
    DriverItem=(Type=Bone,Name="thigh_l")
    FalloffRegionScaleFactors=(PositiveWidth=1,NegativeWidth=1,PositiveHeight=1,NegativeHeight=1)
    FalloffSize=0.3
    FlipHeightScaling=False
    FlipWidthScaling=False
    OptionalParentItem=(Type=Bone,Name="thigh_r")
    RotationOffset=(X=-64.9263,Y=15.2971,Z=-23.6916)
   )
   > RigVMVariableNode(Value=@n9.OutputParam) {variable=Crossed_step_r}
   Subtract = &RigVMFunction_MathDoubleSub(A=1, B=RigVMVariableNode() {variable=thigh_l_back}.Value)
   Clamp = &RigVMFunction_MathDoubleClamp(Maximum=0.8, Value=@Subtract.Result)
   &RigUnit_ParentConstraint(
    AdvancedSettings=(InterpolationType=Average,RotationOrderForFilter=XZY)
    Child=(Type=Bone,Name="skirt_rag_back_l_base")
    Parents=((Item=(Type=Bone,Name="pelvis"),Weight=2),(Item=(Type=Bone,Name="thigh_l"),Weight=1))
    Weight=@Clamp.Result
    ExecutePin=@RigVMFunction_Sequence.D
   )
   n10 = RigVMVariableNode() {variable=thigh_r_back}
   > &RigUnit_ParentConstraint(
    AdvancedSettings=(InterpolationType=Average,RotationOrderForFilter=XZY)
    Child=(Type=Bone,Name="skirt_rag_back_r_base")
    Parents=((Item=(Type=Bone,Name="pelvis"),Weight=2),(Item=(Type=Bone,Name="thigh_r"),Weight=1))
    Weight=&RigVMFunction_MathDoubleClamp(Maximum=0.8, Value=&RigVMFunction_MathDoubleSub(A=1, B=@n10.Value).Result).Result
   )
   &RigUnit_ModifyTransforms(
    ItemToModify="((Item=(Type=Bone,Name=\"skirt_between_front\"),Transform=(Rotation=(X=0,Y=0,Z=0,W=1),Translation=(X=22.4423,Y=-1.8829,Z=-6.753),Scale3D=(X=1,Y=1,Z=1))),(Item=(Type=Bone,Name=\"skirt_between_back\"),Transform=(Rotation=(X=0,Y=-0.0062,Z=0.1735,W=0.9848),Transla ...(truncated)"
    WeightMinimum=0
    ExecutePin=@RigVMFunction_Sequence.E
    Weight=RigVMVariableNode() {variable=thigh_l_up}.Value
   )
   > &RigUnit_ModifyTransforms(
    ItemToModify="((Item=(Type=Bone,Name=\"skirt_between_front\"),Transform=(Rotation=(X=0,Y=0,Z=0,W=1),Translation=(X=0,Y=0,Z=-7),Scale3D=(X=1,Y=1,Z=1))),(Item=(Type=Bone,Name=\"skirt_between_back\"),Transform=(Rotation=(X=0,Y=0,Z=0,W=1),Translation=(X=7,Y=0,Z=0),Scale3D=(X=1, ...(truncated)"
    WeightMinimum=0
    Weight=RigVMVariableNode() {variable=thigh_l_side}.Value
   )
   > &RigUnit_ModifyTransforms(
    ItemToModify="((Item=(Type=Bone,Name=\"skirt_rag_back_l_base\"),Transform=(Rotation=(X=0,Y=0,Z=0,W=1),Translation=(X=0,Y=0,Z=-5.5875),Scale3D=(X=1,Y=1,Z=1))),(Item=(Type=Bone,Name=\"skirt_glute_l\"),Transform=(Rotation=(X=0,Y=0,Z=0,W=1),Translation=(X=0,Y=7.7885,Z=0),Scale3 ...(truncated)"
    WeightMinimum=0
    Weight=RigVMVariableNode() {variable=thigh_l_back}.Value
   )
   > &RigUnit_ModifyTransforms(
    ItemToModify="((Item=(Type=Bone,Name=\"skirt_between_front\"),Transform=(Rotation=(X=0,Y=0,Z=0,W=1),Translation=(X=3.2727,Y=5.9092,Z=10.0941),Scale3D=(X=1,Y=1,Z=1))),(Item=(Type=Bone,Name=\"skirt_rag_front_l_base\"),Transform=(Rotation=(X=-0.0659,Y=0,Z=0,W=0.9978),Translati ...(truncated)"
    WeightMinimum=0
    Weight=RigVMVariableNode() {variable=Crossed_step_l}.Value
   )
   > &RigUnit_ModifyTransforms(
    ItemToModify="((Item=(Type=Bone,Name=\"skirt_between_front\"),Transform=(Rotation=(X=0,Y=0,Z=0,W=1),Translation=(X=30.0409,Y=8.9402,Z=-6.8284),Scale3D=(X=1,Y=1,Z=1))),(Item=(Type=Bone,Name=\"skirt_between_back\"),Transform=(Rotation=(X=0,Y=0,Z=0,W=1),Translation=(X=-0.3802, ...(truncated)"
    WeightMinimum=0
    Weight=RigVMVariableNode() {variable=thigh_r_up}.Value
   )
   > &RigUnit_ModifyTransforms(
    ItemToModify="((Item=(Type=Bone,Name=\"skirt_between_front\"),Transform=(Rotation=(X=0,Y=0,Z=0,W=1),Translation=(X=0,Y=0,Z=7),Scale3D=(X=1,Y=1,Z=1))),(Item=(Type=Bone,Name=\"skirt_between_back\"),Transform=(Rotation=(X=0,Y=0,Z=0,W=1),Translation=(X=-7,Y=0,Z=0),Scale3D=(X=1, ...(truncated)"
    WeightMinimum=0
    Weight=RigVMVariableNode() {variable=thigh_r_side}.Value
   )
   > &RigUnit_ModifyTransforms(
    ItemToModify="((Item=(Type=Bone,Name=\"skirt_rag_back_r_base\"),Transform=(Rotation=(X=0,Y=0,Z=0,W=1),Translation=(X=0,Y=0,Z=-6.981),Scale3D=(X=1,Y=1,Z=1))),(Item=(Type=Bone,Name=\"skirt_glute_r\"),Transform=(Rotation=(X=0,Y=0,Z=0,W=1),Translation=(X=0,Y=-5.9375,Z=0),Scale3 ...(truncated)"
    WeightMinimum=0
    Weight=RigVMVariableNode() {variable=thigh_r_back}.Value
   )
   > &RigUnit_ModifyTransforms(
    ItemToModify=((Item=(Type=Bone,Name="skirt_between_front"),Transform=(Rotation=(X=0,Y=0,Z=0,W=1),Translation=(X=6.099,Y=4.4741,Z=-10.6461),Scale3D=(X=1,Y=1,Z=1))))
    WeightMinimum=0
    Weight=RigVMVariableNode() {variable=Crossed_step_r}.Value
   )
  ]}
]
hierarchy [
 root {loc=<0 0 0>, rot=<0 0 0 1>}
 pelvis {parent=root, loc=<0 -0.5301 98.0571>, rot=<0.0127 0.707 -0.0127 -0.707>}
 spine_01 {parent=pelvis, loc=<6.1208 -0.9328 0>, rot=<0 0 0.034 -0.9994>}
 spine_02 {parent=spine_01, loc=<6.5996 -0.2095 0>, rot=<0 0 0.0238 -0.9997>}
 spine_03 {parent=spine_02, loc=<6.4914 -0.1237 0>, rot=<0 0 0.0149 -0.9999>}
 spine_04 {parent=spine_03, loc=<6.9598 0.0794 0>, rot=<0 0 -0.0132 -0.9999>}
 spine_05 {parent=spine_04, loc=<6.9987 0 0>, rot=<0 0 -0.0015 -1>}
 attach_cape {parent=spine_05, loc=<9.0184 -11.7067 -0.0065>, rot=<0.0616 -0.7044 -0.0616 -0.7044>}
 neck_01 {parent=spine_05, loc=<11.3138 0.1529 0>, rot=<0 0 -0.2003 -0.9797>}
 neck_02 {parent=neck_01, loc=<4.5627 0 0>, rot=<0 0 0.0504 -0.9987>}
 head {parent=neck_02, loc=<5.3914 0 0>, rot=<0 0 0.115 -0.9934>}
 FaceDreadlockFront_r_1 {parent=head, loc=<9.3444 7.1844 4.8013>, rot=<-0.0161 -0.9538 0.0075 0.2998>}
 FaceDreadlockFront_r_2 {parent=FaceDreadlockFront_r_1, loc=<5.2028 0 0>, rot=<0.0112 0.1216 -0.0498 -0.9913>}
 FaceDreadlockFront_r_3 {parent=FaceDreadlockFront_r_2, loc=<4.4155 0 0>, rot=<0.0449 0.2996 -0.006 -0.953>}
 FaceDreadlockFront_r_4 {parent=FaceDreadlockFront_r_3, loc=<5.8795 0 0>, rot=<-0.0358 -0.1854 -0.0352 -0.9814>}
 FaceDreadlockFront_r_5 {parent=FaceDreadlockFront_r_4, loc=<5.2433 0 0>, rot=<-0.016 -0.1362 0.1088 -0.9846>}
 FaceDreadlockFront_r_6 {parent=FaceDreadlockFront_r_5, loc=<8.0009 0 0>, rot=<0.9957 -0.0032 0.0863 0.0328>}
 FaceDreadlockFront_l_1 {parent=head, loc=<5.6287 1.491 -8.0861>, rot=<0.0602 -0.9954 0.0039 -0.0738>}
 FaceDreadlockFront_l_2 {parent=FaceDreadlockFront_l_1, loc=<3.5878 0.0584 -0.0324>, rot=<0.0001 -0.1294 -0.1103 -0.9854>}
 FaceDreadlockFront_l_3 {parent=FaceDreadlockFront_l_2, loc=<5.8375 -0.2845 0.0181>, rot=<0.0065 0.061 0.0036 -0.9981>}
 FaceDreadlockFront_l_4 {parent=FaceDreadlockFront_l_3, loc=<5.8729 0.3115 0.1598>, rot=<-0.0061 -0.0453 -0.03 -0.9985>}
 FaceDreadlockFront_l_5 {parent=FaceDreadlockFront_l_4, loc=<5.057 0.4395 0.8842>, rot=<0.0286 0.1018 -0.1093 -0.9884>}
 FaceDreadlockFront_l_6 {parent=FaceDreadlockFront_l_5, loc=<6.4739 0.688 -0.0918>, rot=<0.1256 0.647 0.1433 -0.7383>}
 FaceDreadlockSide_r_1 {parent=head, loc=<10.2382 5.4669 5.2193>, rot=<0.099 -0.9482 -0.0289 0.3006>}
 FaceDreadlockSide_r_2 {parent=FaceDreadlockSide_r_1, loc=<5.8496 0 0>, rot=<-0.0616 0.3891 -0.0393 -0.9183>}
 FaceDreadlockSide_r_3 {parent=FaceDreadlockSide_r_2, loc=<6.2205 0 0>, rot=<-0.003 -0.0785 -0.152 -0.9853>}
 FaceDreadlockSide_r_4 {parent=FaceDreadlockSide_r_3, loc=<6.507 0 0>, rot=<-0.0048 -0.0189 -0.0557 -0.9983>}
 FaceDreadlockSide_r_5 {parent=FaceDreadlockSide_r_4, loc=<7.5493 0 0>, rot=<0 0 0 1>}
 FACIAL_C_FacialRoot {parent=head, loc=<1.7032 0.1628 -0.0006>, rot=<0.4847 -0.5148 -0.4847 -0.5148>}
 FACIAL_C_Forehead {parent=FACIAL_C_FacialRoot, loc=<-0.0013 -3.0197 8.1096>, rot=<-0.0397 0 0.0001 -0.9992>}
 FACIAL_L_ForeheadIn {parent=FACIAL_C_FacialRoot, loc=<1.832 -3.0213 8.0614>, rot=<-0.0569 -0.0696 0.0046 -0.9959>}
 FACIAL_R_ForeheadIn {parent=FACIAL_C_FacialRoot, loc=<-1.8309 -3.0269 8.054>, rot=<-0.0569 0.0696 -0.0045 -0.9959>}
 FACIAL_L_ForeheadMid {parent=FACIAL_C_FacialRoot, loc=<3.3766 -3.3155 7.5516>, rot=<-0.067 -0.1757 -0.0199 -0.982>}
 FACIAL_R_ForeheadMid {parent=FACIAL_C_FacialRoot, loc=<-3.3751 -3.3206 7.5446>, rot=<-0.067 0.1757 0.0199 -0.982>}
 FACIAL_L_ForeheadOut {parent=FACIAL_C_FacialRoot, loc=<4.732 -3.3528 6.6264>, rot=<-0.1167 -0.256 -0.0779 -0.9565>}
 FACIAL_R_ForeheadOut {parent=FACIAL_C_FacialRoot, loc=<-4.732 -3.3564 6.6213>, rot=<-0.1166 0.256 0.0779 -0.9564>}
 FACIAL_L_EyesackUpper {parent=FACIAL_C_FacialRoot, loc=<3.3514 -2.2441 8.0766>, rot=<0.0565 -0.0847 -0.0638 -0.9928>}
 FACIAL_R_EyesackUpper {parent=FACIAL_C_FacialRoot, loc=<-3.3552 -2.2428 8.074>, rot=<0.0566 0.0847 0.0639 -0.9927>}
 FACIAL_L_EyelidUpperFurrow {parent=FACIAL_C_FacialRoot, loc=<3.1415 -1.966 7.876>, rot=<-0.0199 -0.1516 -0.0781 -0.9852>}
 FACIAL_R_EyelidUpperFurrow {parent=FACIAL_C_FacialRoot, loc=<-3.1454 -1.9654 7.8726>, rot=<-0.0198 0.1517 0.0781 -0.9851>}
 FACIAL_L_EyelidUpperA {parent=FACIAL_C_FacialRoot, loc=<2.9483 -1.0325 5.87>, rot=<-0.1339 0 0 -0.991>}
 FACIAL_R_EyelidUpperA {parent=FACIAL_C_FacialRoot, loc=<-2.9548 -1.0322 5.8768>, rot=<-0.1338 0 0 -0.991>}
 FACIAL_L_Eye {parent=FACIAL_C_FacialRoot, loc=<2.9028 -0.9264 5.9045>, rot=<0.014 0 0 -0.9999>}
 FACIAL_R_Eye {parent=FACIAL_C_FacialRoot, loc=<-2.896 -0.9214 5.9052>, rot=<0.0142 0 0 -0.9999>}
 FACIAL_L_EyelidLowerA {parent=FACIAL_C_FacialRoot, loc=<2.9483 -1.0325 5.87>, rot=<0.2974 0 0 -0.9547>}
 FACIAL_R_EyelidLowerA {parent=FACIAL_C_FacialRoot, loc=<-2.9548 -1.0322 5.8768>, rot=<0.2976 0 0 -0.9547>}
 FACIAL_L_EyeCornerInner {parent=FACIAL_C_FacialRoot, loc=<1.6889 -0.3629 6.9457>, rot=<0.014 0 0 -0.9999>}
 FACIAL_R_EyeCornerInner {parent=FACIAL_C_FacialRoot, loc=<-1.6885 -0.3695 6.949>, rot=<0.0142 0 0 -0.9999>}
 FACIAL_L_EyeCornerOuter {parent=FACIAL_C_FacialRoot, loc=<4.4379 -1.0231 6.5114>, rot=<0.0129 -0.3825 0.0053 -0.9238>}
 FACIAL_R_EyeCornerOuter {parent=FACIAL_C_FacialRoot, loc=<-4.4418 -1.0233 6.5146>, rot=<0.0131 0.3826 -0.0055 -0.9238>}
 FACIAL_L_EyesackLower {parent=FACIAL_C_FacialRoot, loc=<3.4479 0.5161 6.0595>, rot=<-0.0084 -0.14 -0.0203 -0.9899>}
 FACIAL_R_EyesackLower {parent=FACIAL_C_FacialRoot, loc=<-3.4364 0.5164 6.075>, rot=<-0.0082 0.14 0.0202 -0.9899>}
 FACIAL_L_CheekInner {parent=FACIAL_C_FacialRoot, loc=<2.1375 1.1198 6.8555>, rot=<-0.1129 -0.0901 -0.1962 -0.9699>}
 FACIAL_R_CheekInner {parent=FACIAL_C_FacialRoot, loc=<-2.1407 1.1208 6.8615>, rot=<-0.1129 0.0901 0.1962 -0.9699>}
 FACIAL_L_CheekOuter {parent=FACIAL_C_FacialRoot, loc=<4.1089 1.0468 5.4534>, rot=<0.0669 -0.3756 0.1051 -0.9184>}
 FACIAL_R_CheekOuter {parent=FACIAL_C_FacialRoot, loc=<-4.1131 1.0461 5.4544>, rot=<0.0669 0.3757 -0.1052 -0.9183>}
 FACIAL_L_NasolabialBulge {parent=FACIAL_C_FacialRoot, loc=<3.175 4.3892 5.2467>, rot=<0.0628 -0.3772 0.0258 -0.9237>}
 FACIAL_R_NasolabialBulge {parent=FACIAL_C_FacialRoot, loc=<-3.1754 4.3893 5.2457>, rot=<0.0628 0.3772 -0.0258 -0.9236>}
 FACIAL_L_NasolabialFurrow {parent=FACIAL_C_FacialRoot, loc=<2.8512 3.2132 7.858>, rot=<-0.0244 -0.1293 -0.3624 -0.9227>}
 FACIAL_R_NasolabialFurrow {parent=FACIAL_C_FacialRoot, loc=<-2.8531 3.2131 7.8579>, rot=<-0.0244 0.1293 0.3624 -0.9227>}
 FACIAL_L_CheekLower {parent=FACIAL_C_FacialRoot, loc=<3.4431 4.1882 4.6225>, rot=<0.0393 -0.4265 0.06 -0.9016>}
 FACIAL_R_CheekLower {parent=FACIAL_C_FacialRoot, loc=<-3.4435 4.1883 4.6213>, rot=<0.0393 0.4266 -0.06 -0.9016>}
 FACIAL_C_Nose {parent=FACIAL_C_FacialRoot, loc=<-0.0011 0.3755 8.202>, rot=<-0.1244 0 0 -0.9922>}
 FACIAL_C_NoseLower {parent=FACIAL_C_Nose, loc=<0.0002 3.1269 -0.3123>, rot=<0.1304 0 0 -0.9915>}
 FACIAL_L_Nostril {parent=FACIAL_C_Nose, loc=<0.9771 2.439 -0.0619>, rot=<0.0173 0.1305 0.0023 -0.9913>}
 FACIAL_R_Nostril {parent=FACIAL_C_Nose, loc=<-0.9771 2.4393 -0.0617>, rot=<0.0173 -0.1305 -0.0023 -0.9913>}
 FACIAL_C_NoseTip {parent=FACIAL_C_Nose, loc=<0.0003 2.4425 0.5992>, rot=<0.0872 0 0 -0.9962>}
 FACIAL_C_MouthUpper {parent=FACIAL_C_FacialRoot, loc=<-0.0009 4.6109 6.3768>, rot=<0.0053 0 0 -1>}
 FACIAL_C_LipUpper {parent=FACIAL_C_MouthUpper, loc=<0 0.0665 2.3554>, rot=<-0.0872 0 0 -0.9962>}
 FACIAL_L_LipUpper {parent=FACIAL_C_MouthUpper, loc=<0.8406 0.1224 2.3084>, rot=<-0.0778 -0.1601 -0.0379 -0.9833>}
 FACIAL_R_LipUpper {parent=FACIAL_C_MouthUpper, loc=<-0.8399 0.1226 2.3082>, rot=<-0.0778 0.1601 0.0379 -0.9833>}
 FACIAL_L_LipUpperOuter {parent=FACIAL_C_MouthUpper, loc=<1.5824 0.2333 1.9381>, rot=<-0.0984 -0.2293 -0.0983 -0.9634>}
 FACIAL_R_LipUpperOuter {parent=FACIAL_C_MouthUpper, loc=<-1.5816 0.2337 1.9391>, rot=<-0.0984 0.2293 0.0983 -0.9634>}
 FACIAL_L_LipCorner {parent=FACIAL_C_MouthUpper, loc=<2.2422 0.6617 0.9467>, rot=<0 -0.2078 0 -0.9782>}
 FACIAL_R_LipCorner {parent=FACIAL_C_MouthUpper, loc=<-2.242 0.662 0.9469>, rot=<0 0.2078 0 -0.9782>}
 FACIAL_L_Masseter {parent=FACIAL_C_FacialRoot, loc=<6.3777 1.627 3.0362>, rot=<0.0751 0.1129 0.0072 -0.9907>}
 FACIAL_R_Masseter {parent=FACIAL_C_FacialRoot, loc=<-6.3796 1.6263 3.0356>, rot=<0.0751 -0.1129 -0.0072 -0.9907>}
 FACIAL_C_UnderChin {parent=FACIAL_C_FacialRoot, loc=<-0.0002 8.3918 2.8893>, rot=<0.5044 0 0 -0.8634>}
 FACIAL_L_UnderChin {parent=FACIAL_C_FacialRoot, loc=<2.606 7.1107 1.7203>, rot=<0.4076 -0.3544 0.0845 -0.8373>}
 FACIAL_R_UnderChin {parent=FACIAL_C_FacialRoot, loc=<-2.6069 7.1104 1.7197>, rot=<0.4076 0.3545 -0.0845 -0.8373>}
 FACIAL_C_TeethUpper {parent=FACIAL_C_FacialRoot, loc=<-0.0009 5.2885 7.7076>, rot=<0.0053 0 0 -1>}
 FACIAL_C_LowerLipRotation {parent=FACIAL_C_FacialRoot, loc=<0.0001 1.9855 1.1186>, rot=<0.347 0 0 -0.9379>}
 FACIAL_C_MouthLower {parent=FACIAL_C_LowerLipRotation, loc=<-0.0009 -1.4294 5.7007>, rot=<-0.342 0 0 -0.9397>}
 FACIAL_C_LipLower {parent=FACIAL_C_MouthLower, loc=<-0.001 1.5438 1.8146>, rot=<0.0872 0 0 -0.9962>}
 FACIAL_L_LipLower {parent=FACIAL_C_MouthLower, loc=<0.7879 1.504 1.7072>, rot=<0.0927 -0.1294 0.0272 -0.9869>}
 FACIAL_R_LipLower {parent=FACIAL_C_MouthLower, loc=<-0.788 1.5039 1.7072>, rot=<0.0927 0.1294 -0.0272 -0.9869>}
 FACIAL_L_LipLowerOuter {parent=FACIAL_C_MouthLower, loc=<1.4943 1.2593 1.4399>, rot=<0.0913 -0.2321 0.0318 -0.9679>}
 FACIAL_R_LipLowerOuter {parent=FACIAL_C_MouthLower, loc=<-1.4947 1.2592 1.4419>, rot=<0.0913 0.2321 -0.0318 -0.9679>}
 FACIAL_C_TeethLower {parent=FACIAL_C_LowerLipRotation, loc=<-0.0012 -1.7154 7.0984>, rot=<-0.342 0 0 -0.9397>}
 FACIAL_C_Tongue1 {parent=FACIAL_C_TeethLower, loc=<0.0012 -0.0213 -5.1796>, rot=<-0.2588 0 0 -0.9659>}
 FACIAL_C_Tongue2 {parent=FACIAL_C_Tongue1, loc=<-0.0001 0.4504 1.4573>, rot=<0.2588 0 0 -0.9659>}
 FACIAL_C_Tongue3 {parent=FACIAL_C_Tongue2, loc=<-0.0002 0.241 1.3223>, rot=<0.1305 0 0 -0.9914>}
 FACIAL_C_Tongue4 {parent=FACIAL_C_Tongue3, loc=<-0.0004 -0.1413 1.1162>, rot=<0.0436 0 0 -0.999>}
 FACIAL_C_TongueUpper3 {parent=FACIAL_C_Tongue3, loc=<0 -0.3277 -0.201>, rot=<0 0 0 1>}
 FACIAL_C_TongueLower3 {parent=FACIAL_C_Tongue3, loc=<0.0003 0.8349 0.0193>, rot=<0 0 0 1>}
 FACIAL_L_TongueSide3 {parent=FACIAL_C_Tongue3, loc=<1.5226 -0.002 0.0899>, rot=<0 0 0 1>}
 FACIAL_R_TongueSide3 {parent=FACIAL_C_Tongue3, loc=<-1.5227 -0.0017 0.089>, rot=<0 0 0 1>}
 FACIAL_C_TongueUpper2 {parent=FACIAL_C_Tongue2, loc=<0 -0.586 -0.2708>, rot=<0 0 0 1>}
 FACIAL_L_TongueSide2 {parent=FACIAL_C_Tongue2, loc=<1.8159 -0.0219 0.106>, rot=<0 0 0 1>}
 FACIAL_R_TongueSide2 {parent=FACIAL_C_Tongue2, loc=<-1.8157 -0.0216 0.1052>, rot=<0 0 0 1>}
 FACIAL_C_Jaw {parent=FACIAL_C_FacialRoot, loc=<0.0001 1.9855 1.1186>, rot=<0.347 0 0 -0.9379>}
 FACIAL_C_Jawline {parent=FACIAL_C_Jaw, loc=<0.0011 2.5995 7.3681>, rot=<0.2617 0 0 -0.9652>}
 FACIAL_L_ChinSide {parent=FACIAL_C_Jaw, loc=<2.7513 0.4027 7.203>, rot=<-0.2191 -0.3737 0.0505 -0.8999>}
 FACIAL_R_ChinSide {parent=FACIAL_C_Jaw, loc=<-2.7529 0.4036 7.2022>, rot=<-0.2192 0.3737 -0.0506 -0.8999>}
 FACIAL_C_Chin {parent=FACIAL_C_Jaw, loc=<-0.0004 1.075 7.527>, rot=<-0.2541 0 0 -0.9672>}
 FACIAL_L_Ear {parent=FACIAL_C_FacialRoot, loc=<6.6763 -0.3178 -0.2837>, rot=<-0.0193 -0.0008 -0.0262 -0.9995>}
 FACIAL_R_Ear {parent=FACIAL_C_FacialRoot, loc=<-6.6762 -0.3169 -0.2846>, rot=<-0.0193 0.0009 0.0261 -0.9995>}
 FACIAL_L_Jawline {parent=FACIAL_C_FacialRoot, loc=<4.2604 5.6556 3.3688>, rot=<0.2142 0.1575 -0.0554 -0.9624>}
 FACIAL_R_Jawline {parent=FACIAL_C_FacialRoot, loc=<-4.2626 5.6563 3.3689>, rot=<0.2142 -0.1574 0.0555 -0.9624>}
 FACIAL_C_NoseBridge {parent=FACIAL_C_FacialRoot, loc=<-0.0011 -1.7264 8.8942>, rot=<0.0487 0 0 -0.9988>}
 FACIAL_L_EyelidUpperB {parent=FACIAL_C_FacialRoot, loc=<2.9483 -1.0325 5.87>, rot=<-0.2113 0 0 -0.9774>}
 FACIAL_L_EyelidLowerB {parent=FACIAL_C_FacialRoot, loc=<2.9483 -1.0325 5.87>, rot=<0.3305 0 0 -0.9438>}
 FACIAL_R_EyelidLowerB {parent=FACIAL_C_FacialRoot, loc=<-2.9548 -1.0322 5.8768>, rot=<0.339 0 0 -0.9408>}
 FACIAL_R_EyelidUpperB {parent=FACIAL_C_FacialRoot, loc=<-2.9548 -1.0322 5.8768>, rot=<-0.2111 0 0 -0.9775>}
 dreads_right {parent=head, loc=<3.8384 -2.233 8.7188>, rot=<0.0013 -0.9871 -0.0011 -0.16>}
 deform_megaLookUp {parent=head, loc=<14.7325 0.429 0>, rot=<0.0029 -0.7071 -0.0029 -0.7071>}
 back_dread_base {parent=deform_megaLookUp, loc=<0 -11.9306 -9.6122>, rot=<0 0 0 1>}
 dread_spine_r {parent=back_dread_base, loc=<-3.5715 2.7335 -16.4599>, rot=<-0.3526 -0.3489 0.5951 -0.6323>}
 dyn_dread_spine_r_A {parent=dread_spine_r, loc=<6.6294 0.779 -0.7619>, rot=<0.039 -0.1513 0.0237 -0.9874>}
 dyn_dread_spine_r_B {parent=dread_spine_r, loc=<7.4908 -3.9758 -0.6848>, rot=<0.0474 -0.1325 0.2448 -0.9593>}
 dread_spine_l {parent=back_dread_base, loc=<3.5715 2.7335 -16.4595>, rot=<-0.3292 -0.3709 0.5521 -0.6702>}
 dyn_dread_spine_l {parent=dread_spine_l, loc=<5.6453 -0.3024 -0.4283>, rot=<0 -0.0035 0 -1>}
 nape {parent=back_dread_base, loc=<0 -1.1553 -4.8492>, rot=<0 0 0 1>}
 earDreadlock_r_1 {parent=deform_megaLookUp, loc=<-9.7885 -0.3191 -12.1101>, rot=<-0.0016 -0.5848 -0.0023 -0.8112>}
 earDreadlock_r_2 {parent=earDreadlock_r_1, loc=<6.4166 0 0>, rot=<0.0063 -0.1239 0.0558 -0.9907>}
 earDreadlock_r_B1 {parent=earDreadlock_r_2, loc=<5.2047 -0.3286 0.3154>, rot=<-0.0607 -0.221 -0.1253 -0.9653>}
 earDreadlock_r_A1 {parent=earDreadlock_r_2, loc=<4.7452 2.0393 0.37>, rot=<0.0194 0.0255 -0.2341 -0.9717>}
 dread_clavicle_l {parent=deform_megaLookUp, loc=<7.6854 -6.3491 -23.9545>, rot=<-0.1854 -0.3023 0.4654 -0.811>}
 dyn_dread_clavicle_l {parent=dread_clavicle_l, loc=<8.8373 -0.0317 -0.8824>, rot=<-0.0132 -0.23 -0.0559 -0.9715>}
 dread_clavicle_r {parent=deform_megaLookUp, loc=<-7.883 -5.8965 -23.6715>, rot=<-0.226 -0.2766 0.7465 -0.5614>}
 dyn_dread_clavicle_r {parent=dread_clavicle_r, loc=<6.0826 -0.6936 0.3587>, rot=<0.0684 -0.1392 0.0311 -0.9874>}
 sheafBase {parent=deform_megaLookUp, loc=<0 -8.5577 -3.788>, rot=<-0.3446 0.6175 0.3446 -0.6175>}
 sheafCenter {parent=sheafBase, loc=<4.8239 0 0>, rot=<0 0 0 1>}
 asw_dreadFace1 {parent=head, loc=<5.0795 7.3674 7.7755>, rot=<-0.0863 -0.9959 0.0025 0.0259>}
 asw_dreadFace2 {parent=asw_dreadFace1, loc=<15.0442 0 0>, rot=<0 0 0 1>}
 attach_backpack {parent=spine_05, loc=<-0.1159 -13.4592 -0.0032>, rot=<-0.0093 -0.707 0.0093 -0.707>}
 clavicle_l {parent=spine_05, loc=<5.4831 -0.1053 -3.2946>, rot=<0.0465 -0.7289 -0.009 -0.6829>}
 upperarm_l {parent=clavicle_l, loc=<9.8658 -1.2136 -0.4105>, rot=<0.0164 -0.3245 -0.0269 -0.9454>}
 lowerarm_l {parent=upperarm_l, loc=<26.465 -0.8345 0.259>, rot=<0.0036 0.0161 -0.2298 -0.9731>}
 hand_l {parent=lowerarm_l, loc=<22.7355 -0.3202 -0.6106>, rot=<-0.6784 -0.0223 0.0735 -0.7307>}
 thumb_01_l {parent=hand_l, loc=<1.3118 -0.3191 -2.459>, rot=<0.6546 -0.4042 0.0664 -0.6354>}
 thumb_02_l {parent=thumb_01_l, loc=<5.9336 -0.0448 0.0401>, rot=<-0.0037 0.0213 0.298 -0.9543>}
 thumb_03_l {parent=thumb_02_l, loc=<2.2223 0.3682 -0.0907>, rot=<0.0118 -0.013 0.0066 -0.9998>}
 index_metacarpal_l {parent=hand_l, loc=<1.6535 0.0183 -2.96>, rot=<0.0158 -0.0124 -0.0173 -0.9996>}
 index_01_l {parent=index_metacarpal_l, loc=<8.3008 0.0357 -0.0524>, rot=<-0.0044 0.0485 0.0924 -0.9945>}
 index_02_l {parent=index_01_l, loc=<3.3587 -0.3004 0.0452>, rot=<-0.002 0.0027 0.3225 -0.9466>}
 index_03_l {parent=index_02_l, loc=<2.5578 -0.123 -0.0252>, rot=<-0.002 0.0067 0.2558 -0.9667>}
 middle_metacarpal_l {parent=hand_l, loc=<2.1856 -0.1725 -0.5984>, rot=<0.006 0.0168 -0.0329 -0.9993>}
 middle_01_l {parent=middle_metacarpal_l, loc=<7.3275 -0.2447 0.0129>, rot=<-0.0208 0.0042 0.0858 -0.9961>}
 middle_02_l {parent=middle_01_l, loc=<4.0123 0.1401 0.1175>, rot=<-0.0218 0.0258 0.3512 -0.9357>}
 middle_03_l {parent=middle_02_l, loc=<2.6077 0.0534 0.1373>, rot=<0.0082 0.0188 0.1708 -0.9851>}
 ring_metacarpal_l {parent=hand_l, loc=<2.2074 -0.3625 1.3638>, rot=<-0.0284 0.0722 -0.0458 -0.9959>}
 ring_01_l {parent=ring_metacarpal_l, loc=<7.0295 -0.6197 -0.1446>, rot=<-0.0489 -0.0252 0.1892 -0.9804>}
 ring_02_l {parent=ring_01_l, loc=<3.6174 0.4647 -0.1027>, rot=<0 0 0.2806 -0.9598>}
 ring_03_l {parent=ring_02_l, loc=<2.4152 0.114 0.0199>, rot=<0 0 0.1189 -0.9929>}
 pinky_metacarpal_l {parent=hand_l, loc=<1.7616 -0.9753 3.3814>, rot=<-0.0397 0.0814 -0.069 -0.9935>}
 pinky_01_l {parent=pinky_metacarpal_l, loc=<7.1709 -0.4831 -0.1094>, rot=<-0.0654 -0.0035 0.2242 -0.9723>}
 pinky_02_l {parent=pinky_01_l, loc=<2.8274 0.0893 -0.1043>, rot=<0.0037 0.0011 0.2854 -0.9584>}
 pinky_03_l {parent=pinky_02_l, loc=<2.2046 0.0307 -0.0078>, rot=<-0.0002 0.0098 0.1445 -0.9895>}
 deform_wrist_dn_base_l {parent=hand_l, loc=<0 -1.5326 0>, rot=<0 0 0 1>}
 deform_wrist_dn_l {parent=deform_wrist_dn_base_l, loc=<0 0 0>, rot=<0 0 0 1>}
 deform_wrist_up_base_l {parent=hand_l, loc=<0 2.3009 0>, rot=<0 0 0 1>}
 deform_wrist_up_l {parent=deform_wrist_up_base_l, loc=<0 0 0>, rot=<0 0 0 1>}
 weapon_l {parent=hand_l, loc=<-0.1974 -3.4111 0.3811>, rot=<0 0 0 1>}
 hand_attach_l {parent=hand_l, loc=<8.5 -3 0>, rot=<0.7071 0.7071 0 0>}
 lowerarm_twist_02_l {parent=lowerarm_l, loc=<10.6707 0 0>, rot=<0 0 0 1>}
 lowerarm_twist_01_l {parent=lowerarm_l, loc=<18.316 0 0>, rot=<0 0 0 1>}
 deform_elbow_in_base_l {parent=lowerarm_l, loc=<0.1834 3.6062 -0.2572>, rot=<0 0 0 1>}
 deform_elbow_in_l {parent=deform_elbow_in_base_l, loc=<0 0 0>, rot=<0 0 0 1>}
 deform_elbow_out_base_l {parent=lowerarm_l, loc=<0 -4 0>, rot=<0 0 0 1>}
 deform_elbow_out_l {parent=deform_elbow_out_base_l, loc=<0 0 0>, rot=<0 0 0 1>}
 upperarm_twist_01_l {parent=upperarm_l, loc=<7.2135 -0.1134 0>, rot=<0 0 0 1>}
 deform_upperarm_bk_base_l {parent=upperarm_twist_01_l, loc=<-6.4053 -4.3746 0.314>, rot=<0 0 0 1>}
 deform_upperarm_bk_l {parent=deform_upperarm_bk_base_l, loc=<0 0 0>, rot=<0 0 0 1>}
 deform_upperarm_fr_base_l {parent=upperarm_twist_01_l, loc=<-5.6312 5.092 0.3354>, rot=<0 0 0 1>}
 deform_upperarm_fr_l {parent=deform_upperarm_fr_base_l, loc=<0 0 0>, rot=<0 0 0 1>}
 deform_upperarm_up_base_l {parent=upperarm_twist_01_l, loc=<-8.891 0.4825 3.9517>, rot=<0 0 0 1>}
 deform_upperarm_up_l {parent=deform_upperarm_up_base_l, loc=<0 0 0>, rot=<0 0 0 1>}
 upperarm_twist_02_l {parent=upperarm_l, loc=<19.7998 -0.1134 0>, rot=<0 0 0 1>}
 clavicle_r {parent=spine_05, loc=<5.483 -0.1053 3.2946>, rot=<0.7289 0.0465 -0.6829 0.009>}
 upperarm_r {parent=clavicle_r, loc=<-9.8658 1.2136 0.4105>, rot=<0.0164 -0.3245 -0.0269 -0.9454>}
 lowerarm_r {parent=upperarm_r, loc=<-26.465 0.8345 -0.259>, rot=<0.0036 0.0161 -0.2298 -0.9731>}
 hand_r {parent=lowerarm_r, loc=<-22.7355 0.3202 0.6106>, rot=<-0.6784 -0.0223 0.0735 -0.7307>}
 thumb_01_r {parent=hand_r, loc=<-1.3118 0.3191 2.459>, rot=<0.6546 -0.4042 0.0664 -0.6354>}
 thumb_02_r {parent=thumb_01_r, loc=<-5.9336 0.0448 -0.0401>, rot=<-0.0037 0.0213 0.298 -0.9543>}
 thumb_03_r {parent=thumb_02_r, loc=<-2.2223 -0.3682 0.0907>, rot=<0.0118 -0.013 0.0066 -0.9998>}
 index_metacarpal_r {parent=hand_r, loc=<-1.6535 -0.0183 2.96>, rot=<0.0158 -0.0124 -0.0173 -0.9996>}
 index_01_r {parent=index_metacarpal_r, loc=<-8.3008 -0.0357 0.0524>, rot=<-0.0044 0.0485 0.0924 -0.9945>}
 index_02_r {parent=index_01_r, loc=<-3.3587 0.3004 -0.0452>, rot=<-0.002 0.0027 0.3225 -0.9466>}
 index_03_r {parent=index_02_r, loc=<-2.5578 0.123 0.0252>, rot=<-0.002 0.0067 0.2558 -0.9667>}
 middle_metacarpal_r {parent=hand_r, loc=<-2.1856 0.1725 0.5984>, rot=<0.006 0.0168 -0.0329 -0.9993>}
 middle_01_r {parent=middle_metacarpal_r, loc=<-7.3275 0.2447 -0.0129>, rot=<-0.0208 0.0042 0.0858 -0.9961>}
 middle_02_r {parent=middle_01_r, loc=<-4.0123 -0.1401 -0.1175>, rot=<-0.0218 0.0258 0.3512 -0.9357>}
 middle_03_r {parent=middle_02_r, loc=<-2.6077 -0.0534 -0.1373>, rot=<0.0082 0.0188 0.1708 -0.9851>}
 ring_metacarpal_r {parent=hand_r, loc=<-2.2074 0.3625 -1.3638>, rot=<-0.0284 0.0722 -0.0458 -0.9959>}
 ring_01_r {parent=ring_metacarpal_r, loc=<-7.0295 0.6197 0.1446>, rot=<-0.0489 -0.0252 0.1892 -0.9804>}
 ring_02_r {parent=ring_01_r, loc=<-3.6174 -0.4647 0.1027>, rot=<0 0 0.2806 -0.9598>}
 ring_03_r {parent=ring_02_r, loc=<-2.4152 -0.114 -0.0199>, rot=<0 0 0.1189 -0.9929>}
 pinky_metacarpal_r {parent=hand_r, loc=<-1.7616 0.9753 -3.3814>, rot=<-0.0397 0.0814 -0.069 -0.9935>}
 pinky_01_r {parent=pinky_metacarpal_r, loc=<-7.1709 0.4831 0.1094>, rot=<-0.0654 -0.0035 0.2242 -0.9723>}
 pinky_02_r {parent=pinky_01_r, loc=<-2.8274 -0.0893 0.1043>, rot=<0.0037 0.0011 0.2854 -0.9584>}
 pinky_03_r {parent=pinky_02_r, loc=<-2.2046 -0.0307 0.0078>, rot=<-0.0002 0.0098 0.1445 -0.9895>}
 deform_wrist_dn_base_r {parent=hand_r, loc=<-0.0001 1.5328 0>, rot=<0 0 0 1>}
 deform_wrist_dn_r {parent=deform_wrist_dn_base_r, loc=<0 0 0>, rot=<0 0 0 1>}
 deform_wrist_up_base_r {parent=hand_r, loc=<-0.0002 -2.3007 -0.0001>, rot=<0 0 0 1>}
 deform_wrist_up_r {parent=deform_wrist_up_base_r, loc=<0 0 0>, rot=<0 0 0 1>}
 weapon_r {parent=hand_r, loc=<0.1974 3.4111 -0.3811>, rot=<0 0 0 1>}
 hand_attach_r {parent=hand_r, loc=<-8.5001 2.9999 0>, rot=<0 0 -0.7071 -0.7071>}
 lowerarm_twist_02_r {parent=lowerarm_r, loc=<-10.6707 0 0>, rot=<0 0 0 1>}
 lowerarm_twist_01_r {parent=lowerarm_r, loc=<-18.316 0 0>, rot=<0 0 0 1>}
 deform_elbow_in_base_r {parent=lowerarm_r, loc=<-0.1834 -3.6062 0.2572>, rot=<0 0 0 1>}
 deform_elbow_in_r {parent=deform_elbow_in_base_r, loc=<0 0 0>, rot=<0 0 0 1>}
 deform_elbow_out_base_r {parent=lowerarm_r, loc=<0 4 -0.0001>, rot=<0 0 0 1>}
 deform_elbow_out_r {parent=deform_elbow_out_base_r, loc=<0 0 0>, rot=<0 0 0 1>}
 upperarm_twist_01_r {parent=upperarm_r, loc=<-7.2135 0.1134 0>, rot=<0 0 0 1>}
 deform_upperarm_bk_base_r {parent=upperarm_twist_01_r, loc=<6.4053 4.3746 -0.314>, rot=<0 0 0 1>}
 deform_upperarm_bk_r {parent=deform_upperarm_bk_base_r, loc=<0 0 0>, rot=<0 0 0 1>}
 deform_upperarm_fr_base_r {parent=upperarm_twist_01_r, loc=<5.6312 -5.092 -0.3354>, rot=<0 0 0 1>}
 deform_upperarm_fr_r {parent=deform_upperarm_fr_base_r, loc=<0 0 0>, rot=<0 0 0 1>}
 deform_upperarm_up_base_r {parent=upperarm_twist_01_r, loc=<8.891 -0.4825 -3.9517>, rot=<0 0 0 1>}
 deform_upperarm_up_r {parent=deform_upperarm_up_base_r, loc=<0 0 0>, rot=<0 0 0 1>}
 upperarm_twist_02_r {parent=upperarm_r, loc=<-19.7998 0.1134 0>, rot=<0 0 0 1>}
 deform_pec_aim_base_l {parent=spine_04, loc=<3.2456 9.3309 -1.2237>, rot=<0.0603 -0.4033 0.3318 -0.8507>}
 deform_pec_aim_l {parent=deform_pec_aim_base_l, loc=<0 0 0>, rot=<0 0 0 1>}
 deform_pec_base_l {parent=deform_pec_aim_l, loc=<9.3665 0 0>, rot=<0 0 0 1>}
 deform_pec_l {parent=deform_pec_base_l, loc=<0 0 0>, rot=<0 0 0 1>}
 deform_pec_aim_base_r {parent=spine_04, loc=<3.2452 9.3309 1.2237>, rot=<-0.4033 -0.0603 0.8507 0.3318>}
 deform_pec_aim_r {parent=deform_pec_aim_base_r, loc=<0 0 0>, rot=<0 0 0 1>}
 deform_pec_base_r {parent=deform_pec_aim_r, loc=<-9.3667 -0.0003 -0.0004>, rot=<0 0 0 1>}
 deform_pec_r {parent=deform_pec_base_r, loc=<0 0 0>, rot=<0 0 0 1>}
 thigh_l {parent=pelvis, loc=<-1.0855 -0.2109 -8.7053>, rot=<0.0009 0.0355 0.0155 -0.9993>}
 calf_l {parent=thigh_l, loc=<-42.5284 -0.0019 -0.7297>, rot=<0.0006 -0.0161 -0.0391 -0.9991>}
 foot_l {parent=calf_l, loc=<-46.7402 -2.5014 -0.5919>, rot=<0.0427 -0.0238 0.0353 -0.9982>}
 ball_l {parent=foot_l, loc=<-3.5658 10.8066 -0.1545>, rot=<0.0001 -0.0001 -0.7063 -0.7079>}
 calf_twist_02_l {parent=calf_l, loc=<-19.1546 -1.7943 -0.2638>, rot=<0.0237 0.0095 -0.0035 -0.9997>}
 calf_twist_01_l {parent=calf_l, loc=<-35.8484 -1.9196 -0.5793>, rot=<0.0237 0.0095 -0.0035 -0.9997>}
 deform_knee_in_base_l {parent=calf_l, loc=<0.8031 -5.0231 0.2456>, rot=<0 0 0 1>}
 deform_knee_in_l {parent=deform_knee_in_base_l, loc=<0 0 0>, rot=<0 0 0 1>}
 deform_knee_out_base_l {parent=calf_l, loc=<0 4 0>, rot=<0 0 0 1>}
 deform_knee_out_l {parent=deform_knee_out_base_l, loc=<0 0 0>, rot=<0 0 0 1>}
 thigh_twist_01_l {parent=thigh_l, loc=<-27.952 0 0>, rot=<0.0013 0 0 -1>}
 deform_groin_base_l {parent=thigh_l, loc=<0.4801 5.6776 -0.2486>, rot=<0 0 0 1>}
 deform_groin_l {parent=deform_groin_base_l, loc=<0 0 0>, rot=<0 0 0 1>}
 skirt_rag_leg_l_1 {parent=thigh_l, loc=<-17.2533 3.2439 -6.9814>, rot=<0.0336 -0.2589 0.9652 0.0115>}
 skirt_rag_leg_l_2 {parent=skirt_rag_leg_l_1, loc=<4.9683 3.7154 0.8607>, rot=<-0.2587 0.0062 -0.0233 -0.9656>}
 skirt_rag_leg_l_3 {parent=skirt_rag_leg_l_2, loc=<5.8184 0 0>, rot=<0 0 0 1>}
 skirt_rag_leg_l_4 {parent=skirt_rag_leg_l_3, loc=<6.9895 0 0>, rot=<0 0 0 -1>}
 skirt_rag_back_l_base {parent=thigh_l, loc=<-18.3936 -7.9343 -0.543>, rot=<0.0291 0.6277 0.7776 -0.0204>}
 skirt_rag_back_l_1 {parent=skirt_rag_back_l_base, loc=<0 0 0>, rot=<0 0 0 1>}
 skirt_rag_back_l_2 {parent=skirt_rag_back_l_1, loc=<8.6608 0 0>, rot=<0 0 0 1>}
 skirt_rag_front_l_base {parent=thigh_l, loc=<-16.7706 8.4368 1.0321>, rot=<0.0179 -0.6619 0.7492 0.0143>}
 skirt_rag_front_l_1 {parent=skirt_rag_front_l_base, loc=<0 0 0>, rot=<0 0 0 1>}
 skirt_rag_front_l_2 {parent=skirt_rag_front_l_1, loc=<6.959 0 0>, rot=<0 0 0 1>}
 skirt_rag_front_l_3 {parent=skirt_rag_front_l_2, loc=<10.1894 0 0>, rot=<0 0 0 1>}
 thigh_r {parent=pelvis, loc=<-1.0855 -0.2109 8.7053>, rot=<0.0355 -0.0009 0.9993 0.0155>}
 calf_r {parent=thigh_r, loc=<42.5284 0.0019 0.7297>, rot=<0.0006 -0.0161 -0.0391 -0.9991>}
 foot_r {parent=calf_r, loc=<46.7402 2.5014 0.592>, rot=<0.0427 -0.0238 0.0353 -0.9982>}
 ball_r {parent=foot_r, loc=<3.5658 -10.8066 0.1545>, rot=<0.0001 -0.0001 -0.7063 -0.7079>}
 calf_twist_02_r {parent=calf_r, loc=<19.1546 1.7943 0.2638>, rot=<0.0237 0.0095 -0.0035 -0.9997>}
 calf_twist_01_r {parent=calf_r, loc=<35.8484 1.9197 0.5794>, rot=<0.0237 0.0095 -0.0035 -0.9997>}
 deform_knee_in_base_r {parent=calf_r, loc=<-0.8031 5.0231 -0.2457>, rot=<0 0 0 1>}
 deform_knee_in_r {parent=deform_knee_in_base_r, loc=<0 0 0>, rot=<0 0 0 1>}
 deform_knee_out_base_r {parent=calf_r, loc=<0 -4 0>, rot=<0 0 0 1>}
 deform_knee_out_r {parent=deform_knee_out_base_r, loc=<0 0 0>, rot=<0 0 0 1>}
 kneeArmor_r_base {parent=calf_r, loc=<-1.5804 -4.2373 0>, rot=<0 0 0 1>}
 kneeArmor_r {parent=kneeArmor_r_base, loc=<0 0 0>, rot=<0 0 0 1>}
 thigh_twist_01_r {parent=thigh_r, loc=<27.952 0 0>, rot=<0.0013 0 0 -1>}
 deform_groin_base_r {parent=thigh_r, loc=<-0.4801 -5.6776 0.2486>, rot=<0 0 0 -1>}
 deform_groin_r {parent=deform_groin_base_r, loc=<0 0 0>, rot=<0 0 0 1>}
 skirt_rag_leg_r_B1 {parent=thigh_r, loc=<17.9329 2.3952 7.0327>, rot=<0.9905 0.0041 -0.0431 0.1304>}
 skirt_rag_leg_r_B2 {parent=skirt_rag_leg_r_B1, loc=<7.0982 0 0>, rot=<0 0 0 -1>}
 skirt_rag_leg_r_B3 {parent=skirt_rag_leg_r_B2, loc=<5.9706 0 0>, rot=<0 0 0 1>}
 skirt_rag_back_r_base {parent=thigh_r, loc=<18.1733 8.1652 -2.5711>, rot=<-0.7068 -0.0234 0.0268 -0.7065>}
 skirt_rag_back_r_1 {parent=skirt_rag_back_r_base, loc=<0 0 0>, rot=<0 0 0 1>}
 skirt_rag_back_r_2 {parent=skirt_rag_back_r_1, loc=<8.6608 0 0>, rot=<0 0 0 1>}
 skirt_rag_front_r_base {parent=thigh_r, loc=<16.5938 -8.435 -3.5184>, rot=<0.7066 -0.0246 -0.0188 -0.7069>}
 skirt_rag_front_r_1 {parent=skirt_rag_front_r_base, loc=<0 0 0>, rot=<0 0 0 1>}
 skirt_rag_front_r_2 {parent=skirt_rag_front_r_1, loc=<5.3962 0 0>, rot=<0 0 0 1>}
 skirt_rag_front_r_3 {parent=skirt_rag_front_r_2, loc=<10.3645 0 0>, rot=<0 0 0 1>}
 skirt_rag_leg_r_A1_base {parent=thigh_r, loc=<16.7849 -4.9453 6.3374>, rot=<0.945 -0.0098 -0.0298 -0.3254>}
 skirt_rag_leg_r_A1 {parent=skirt_rag_leg_r_A1_base, loc=<0 0 0>, rot=<0 0 0 1>}
 skirt_rag_leg_r_A2 {parent=skirt_rag_leg_r_A1, loc=<5.3962 0 0>, rot=<0.0489 -0.001 0.0308 -0.9983>}
 skirt_rag_leg_r_A3 {parent=skirt_rag_leg_r_A2, loc=<7.7725 0 0>, rot=<-0.0361 0 0.0209 -0.9991>}
 deform_glute_base_l {parent=pelvis, loc=<-7.3466 -9.0213 -9.1245>, rot=<0.0172 0.031 -0.4477 -0.8935>}
 deform_glute_l {parent=deform_glute_base_l, loc=<0 0 0>, rot=<0 0 0 1>}
 skirt_glute_l {parent=deform_glute_l, loc=<0 0 0>, rot=<-0.5726 0.3384 0.5755 0.4758>}
 deform_glute_base_r {parent=pelvis, loc=<-7.3464 -9.0213 9.1244>, rot=<0.031 -0.0172 0.8935 -0.4477>}
 deform_glute_r {parent=deform_glute_base_r, loc=<0 0 0>, rot=<0 0 0 1>}
 skirt_glute_r {parent=deform_glute_r, loc=<0 0 0>, rot=<-0.5726 0.3384 0.5755 0.4758>}
 skirt_between_front_base {parent=pelvis, loc=<-14.9854 8.4925 0>, rot=<0 0 0 1>}
 skirt_between_front {parent=skirt_between_front_base, loc=<0 0 0>, rot=<0 0 0 1>}
 skirt_between_back_base {parent=pelvis, loc=<-16.91 -8.061 0>, rot=<0 -0.7071 0 -0.7071>}
 skirt_between_back {parent=skirt_between_back_base, loc=<0 0 0>, rot=<0 0 0 1>}
 attach {parent=root, loc=<0 0 0>, rot=<0 0 0 1>}
 ik_hand_root {parent=root, loc=<0 0 0>, rot=<0 0 0 1>}
 ik_hand_gun {parent=ik_hand_root, loc=<-48.1743 3.3545 104.4381>, rot=<0.6276 0.1309 0.4183 -0.6434>}
 ik_hand_r {parent=ik_hand_gun, loc=<0 0 0>, rot=<0 0 0 1>}
 ik_hand_l {parent=ik_hand_gun, loc=<59.3271 67.6997 34.3524>, rot=<-0.6158 -0.7027 -0.3565 0>}
 ik_foot_root {parent=root, loc=<0 0 0>, rot=<0 0 0 1>}
 ik_foot_l {parent=ik_foot_root, loc=<14.8262 -7.3545 8.296>, rot=<0.0357 0.7045 0.0284 -0.7082>}
 ik_foot_r {parent=ik_foot_root, loc=<-14.8262 -7.3545 8.296>, rot=<0.7082 0.0284 -0.7045 0.0357>}
 attach_fp {parent=root, loc=<0 0 0>, rot=<0 0 0 1>}
 camera_root {parent=root, loc=<0 0 0>, rot=<0 0 0 1>}
 camera {parent=camera_root, loc=<0 0 0>, rot=<0 0 0 1>}
 warp_point_01 {parent=root, loc=<0 0 0>, rot=<0 0 0 1>}
 warp_point_02 {parent=root, loc=<0 0 0>, rot=<0 0 0 1>}
 warp_point_03 {parent=root, loc=<0 0 0>, rot=<0 0 0 1>}
]
variables [
 Crossed_step_l {type=double, default=0}
 Crossed_step_r {type=double, default=0}
 thigh_l_back {type=double, default=0}
 thigh_l_side {type=double, default=0}
 thigh_l_up {type=double, default=0}
 thigh_r_back {type=double, default=0}
 thigh_r_side {type=double, default=0}
 thigh_r_up {type=double, default=0}
]
