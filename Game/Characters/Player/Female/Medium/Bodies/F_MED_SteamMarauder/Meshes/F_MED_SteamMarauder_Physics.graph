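# Physics asset for the medium female SteamMarauder player body, covering the lower body and its hanging skirt rags.
# Capsules cover pelvis, thighs and calves, while boxes form rag chains along the front, back and legs.
# Front and back chains hang from small base spheres, and leg chains attach to the thighs.
# Most rag joints use tight angular limits, some with locked axes; thigh-to-pelvis and two right-leg rag joints swing and twist freely.
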PhysicsAsset /Game/Characters/Player/Female/Medium/Bodies/F_MED_SteamMarauder/Meshes/F_MED_SteamMarauder_Physics
bodies {
 calf_l [{shapeType=Capsule, length=31.5546, radius=6.5954, shapeName=thigh_l_capsule, center=<-20.3318 -1.6877 0>, rotation=<rot -90 -90 89.9997>}]
 calf_r [{shapeType=Capsule, length=31.5546, radius=6.5954, shapeName=thigh_l_capsule, center=<20.3318 1.6877 0>, rotation=<rot -90 -172.4477 -7.5526>}]
 pelvis [{shapeType=Capsule, length=11.7311, radius=10.405, shapeName=pelvis_capsule, center=<-6.5741 -1.5477 0>}]
 skirt_rag_back_l_1 [{
   shapeType=Box
   extentX=7.6681
   extentY=8.3199
   extentZ=2.0299
   shapeName=skirt_rag_front_l_1_box
   center=<4.1305 -0.0052 0.1507>
   rotation=<rot 0.0001 0 -1.9854>
  }]
 skirt_rag_back_l_2 [{
   shapeType=Box
   extentX=11.0809
   extentY=6.0806
   extentZ=2.3027
   shapeName=skirt_rag_front_l_1_box
   center=<5.8963 0 0.0921>
   rotation=<rot 0 0 -5>
  }]
 skirt_rag_back_l_base [{shapeType=Sphere, radius=2.7476, shapeName=skirt_rag_back_r_base_sphere}]
 skirt_rag_back_r_1 [{
   shapeType=Box
   extentX=10.7214
   extentY=8.3199
   extentZ=4.4394
   shapeName=skirt_rag_front_l_1_box
   center=<5.1375 0 0.25>
   rotation=<rot 4.8244 0 0>
  }]
 skirt_rag_back_r_2 [{shapeType=Box, extentX=13.0789, extentY=9.1978, extentZ=3.3876, shapeName=skirt_rag_front_l_1_box, center=<5.7533 -0.2888 0>}]
 skirt_rag_back_r_base [{shapeType=Sphere, radius=2.7476, shapeName=skirt_rag_back_r_base_sphere}]
 skirt_rag_front_l_1 [{shapeType=Box, extentX=7.5, extentY=5.4287, extentZ=2.7741, shapeName=skirt_rag_front_l_1_box, center=<3.263 0 -0.3676>}]
 skirt_rag_front_l_2 [{shapeType=Box, extentX=9.69, extentY=8.3199, extentZ=3.0358, shapeName=skirt_rag_front_l_1_box, center=<5.1393 0 -0.3676>}]
 skirt_rag_front_l_3 [{shapeType=Box, extentX=9.69, extentY=8.3199, extentZ=2.4028, shapeName=skirt_rag_front_l_1_box, center=<5.1393 0 -0.3676>}]
 skirt_rag_front_l_base [{shapeType=Sphere, radius=2.7476, shapeName=skirt_rag_back_r_base_sphere}]
 skirt_rag_front_r_1 [{shapeType=Box, extentX=5.7146, extentY=5.7996, extentZ=1.6974, shapeName=skirt_rag_front_l_1_box, center=<2.5507 0 0>}]
 skirt_rag_front_r_2 [{shapeType=Box, extentX=9.942, extentY=8.3199, extentZ=1.6974, shapeName=skirt_rag_front_l_1_box, center=<5.1108 0 0>}]
 skirt_rag_front_r_3 [{shapeType=Box, extentX=9.3187, extentY=6.6609, extentZ=1.6974, shapeName=skirt_rag_front_l_1_box, center=<4.6367 0 0>}]
 skirt_rag_front_r_base [{shapeType=Sphere, radius=2.7476, shapeName=skirt_rag_back_r_base_sphere}]
 skirt_rag_leg_l_1 [{
   shapeType=Box
   extentX=7.5
   extentY=5.4385
   extentZ=1.2402
   shapeName=skirt_rag_front_l_1_box
   center=<2.7762 0.0455 -0.2582>
   rotation=<rot -0.0001 -0.0002 -10>
  }]
 skirt_rag_leg_l_2 [{
   shapeType=Box
   extentX=6.263
   extentY=8.3199
   extentZ=1.6974
   shapeName=skirt_rag_front_l_1_box
   center=<2.4708 0 0>
   rotation=<rot 0 -0.0001 -6.4123>
  }]
 skirt_rag_leg_l_3 [{shapeType=Box, extentX=7.5, extentY=8.3199, extentZ=1.6974, shapeName=skirt_rag_front_l_1_box, center=<3.745 0 0>}]
 skirt_rag_leg_l_4 [{shapeType=Box, extentX=7.5, extentY=8.3199, extentZ=1.6974, shapeName=skirt_rag_front_l_1_box, center=<3.745 0 0>}]
 skirt_rag_leg_r_A1 [{shapeType=Box, extentX=5.4618, extentY=6.3784, extentZ=2.2371, shapeName=skirt_rag_front_l_1_box, center=<2.8354 0 0>}]
 skirt_rag_leg_r_A2 [{
   shapeType=Box
   extentX=7.5
   extentY=7.161
   extentZ=2.9374
   shapeName=skirt_rag_front_l_1_box
   center=<3.7449 -0.0027 0.0372>
   rotation=<rot 0.2467 2.3078 -4.0962>
  }]
 skirt_rag_leg_r_A3 [{
   shapeType=Box
   extentX=9.8387
   extentY=3.7391
   extentZ=1.6974
   shapeName=skirt_rag_front_l_1_box
   center=<5.033 0.1028 0.2694>
   rotation=<rot -0.0002 4.6055 -0.2>
  }]
 skirt_rag_leg_r_B1 [{shapeType=Box, extentX=8.4276, extentY=8.3199, extentZ=3.844, shapeName=skirt_rag_front_l_1_box, center=<3.7766 0 0.3778>}]
 skirt_rag_leg_r_B2 [{shapeType=Box, extentX=9.68, extentY=8.3199, extentZ=3.9681, shapeName=skirt_rag_front_l_1_box, center=<3.2708 0 0>}]
 skirt_rag_leg_r_B3 [{shapeType=Box, extentX=9.1392, extentY=5.3161, extentZ=3.567, shapeName=skirt_rag_front_l_1_box, center=<4.6784 -0.3328 0>}]
 thigh_l [{
   shapeType=Capsule
   length=30.8953
   radius=7.2546
   shapeName=thigh_l_capsule
   center=<-22.6517 -0.0543 -0.1914>
   rotation=<rot -90 -179.9703 179.9701>
  }]
 thigh_r [{
   shapeType=Capsule
   length=31.2018
   radius=6.9481
   shapeName=thigh_l_capsule
   center=<22.6517 -0.2051 -0.2143>
   rotation=<rot -90 -160.5288 -19.4715>
  }]
}
constraints [
 {
  bone1Name=thigh_l
  bone2Name=pelvis
  swing1LimitDegrees=45
  swing1Motion=Free
  swing2LimitDegrees=45
  swing2Motion=Free
  twistLimitDegrees=45
  twistMotion=Free
 }
 {bone1Name=skirt_rag_front_l_3, bone2Name=skirt_rag_front_l_2, swing1LimitDegrees=25, swing2LimitDegrees=25, twistLimitDegrees=25}
 {bone1Name=skirt_rag_front_l_2, bone2Name=skirt_rag_front_l_1, swing1LimitDegrees=30, swing2LimitDegrees=30, twistLimitDegrees=25}
 {
  bone1Name=thigh_r
  bone2Name=pelvis
  swing1LimitDegrees=45
  swing1Motion=Free
  swing2LimitDegrees=45
  swing2Motion=Free
  twistLimitDegrees=45
  twistMotion=Free
 }
 {bone1Name=skirt_rag_leg_l_4, bone2Name=skirt_rag_leg_l_3, swing1LimitDegrees=15, swing2LimitDegrees=25, twistLimitDegrees=15}
 {bone1Name=skirt_rag_leg_l_3, bone2Name=skirt_rag_leg_l_2, swing1LimitDegrees=15, swing2LimitDegrees=25, twistLimitDegrees=15}
 {bone1Name=skirt_rag_leg_l_2, bone2Name=skirt_rag_leg_l_1, swing1LimitDegrees=30, swing2LimitDegrees=30, twistLimitDegrees=12}
 {bone1Name=skirt_rag_leg_l_1, bone2Name=thigh_l, swing1LimitDegrees=5, swing2LimitDegrees=5, twistLimitDegrees=5}
 {bone1Name=skirt_rag_back_l_2, bone2Name=skirt_rag_back_l_1, swing1LimitDegrees=15, swing2LimitDegrees=25, twistLimitDegrees=25}
 {bone1Name=skirt_rag_back_r_2, bone2Name=skirt_rag_back_r_1, swing1LimitDegrees=25, swing2LimitDegrees=25, twistLimitDegrees=25}
 {bone1Name=skirt_rag_leg_r_B3, bone2Name=skirt_rag_leg_r_B2, swing1LimitDegrees=15, swing2LimitDegrees=25, twistLimitDegrees=15}
 {
  bone1Name=skirt_rag_leg_r_B2
  bone2Name=skirt_rag_leg_r_B1
  swing1LimitDegrees=15
  swing2LimitDegrees=25
  swing2Motion=Locked
  twistLimitDegrees=15
 }
 {
  bone1Name=skirt_rag_leg_r_B1
  bone2Name=thigh_r
  swing1LimitDegrees=30
  swing2LimitDegrees=30
  swing2Motion=Locked
  twistLimitDegrees=15
  twistMotion=Locked
 }
 {bone1Name=skirt_rag_leg_r_A3, bone2Name=skirt_rag_leg_r_A2, swing1LimitDegrees=15, swing2LimitDegrees=25, twistLimitDegrees=15}
 {bone1Name=skirt_rag_front_r_3, bone2Name=skirt_rag_front_r_2, swing1LimitDegrees=15, swing2LimitDegrees=25, twistLimitDegrees=25}
 {bone1Name=skirt_rag_front_r_2, bone2Name=skirt_rag_front_r_1, swing1LimitDegrees=15, swing2LimitDegrees=15, twistLimitDegrees=25}
 {
  bone1Name=skirt_rag_back_r_1
  bone2Name=skirt_rag_back_r_base
  swing1LimitDegrees=15
  swing1Motion=Locked
  swing2LimitDegrees=15
  twistLimitDegrees=15
 }
 {bone1Name=skirt_rag_back_l_1, bone2Name=skirt_rag_back_l_base, swing1LimitDegrees=25, swing2LimitDegrees=25, twistLimitDegrees=5}
 {bone1Name=skirt_rag_front_r_1, bone2Name=skirt_rag_front_r_base, swing1LimitDegrees=15, swing2LimitDegrees=15, twistLimitDegrees=15}
 {bone1Name=skirt_rag_front_l_1, bone2Name=skirt_rag_front_l_base, swing1LimitDegrees=20, swing2LimitDegrees=20, twistLimitDegrees=15}
 {
  bone1Name=skirt_rag_leg_r_A2
  bone2Name=skirt_rag_leg_r_A1
  swing1LimitDegrees=45
  swing1Motion=Free
  swing2LimitDegrees=45
  swing2Motion=Free
  twistLimitDegrees=45
  twistMotion=Free
 }
 {
  bone1Name=skirt_rag_leg_r_A3
  bone2Name=skirt_rag_leg_r_A1
  swing1LimitDegrees=45
  swing1Motion=Free
  swing2LimitDegrees=45
  swing2Motion=Free
  twistLimitDegrees=45
  twistMotion=Free
 }
 {bone1Name=skirt_rag_leg_r_A1, bone2Name=thigh_r, swing1LimitDegrees=15, swing2LimitDegrees=25, twistLimitDegrees=15, twistMotion=Locked}
]
