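# Post-pose AnimBlueprint for the F_MED_SteamMarauder female player body, layering cosmetic secondary motion onto copied poses.
# Copies the attached parent's pose and blends in facial bones from the Head part, cached when customization loads.
# The result feeds a Control Rig, then RigidBody physics in root-bone space.
# Component-space fixups then constrain a right-leg skirt rag bone to the thigh and nudge the backpack attach bone.
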AnimBlueprint /Game/Characters/Player/Female/Medium/Bodies/F_MED_SteamMarauder/Meshes/F_MED_SteamMarauder_AnimBP
target_skeleton=&/F_MED_SteamMarauder_Skeleton
graphs [
 AnimGraph {nodes [
   n1 = Get Head() -> (Head:Skeletal Mesh Component Object Reference)
   n2 = Control Rig(Source=Layered blend per bone(
     BasePose=Copy Pose From Mesh() {bUseAttachedParent=true}
     BlendPoses_0=Copy Pose From Mesh(SourceMeshComponent=@n1) {bUseAttachedParent=true}
    ) {layerSetup [{branchFilters [{blendDepth=1, boneName=FACIAL_C_FacialRoot}]}]}) {
    alphaBoolBlend {blendOption=Linear}
    controlRigAssetReference {blueprintRigClass=&/F_MED_SteamMarauder_CtrlRig.F_MED_SteamMarauder_CtrlRig_C}
   }
   n3 = RigidBody(ComponentPose=Local To Component(LocalPose=@n2)) {
    bClampLinearTranslationLimitToRefPose=true
    overridePhysicsAsset=&/F_MED_SteamMarauder_Physics
    simulationSpace=BaseBoneSpace
    alphaBoolBlend {blendOption=Linear}
    baseBoneRef {boneName=root}
    componentAppliedLinearAccClamp=<3000 3000 3000>
    componentLinearAccScale=<1 1 1>
    componentLinearVelScale=<1 1 1>
    simSpaceSettings {maxAngularAcceleration=150, maxAngularVelocity=3, maxLinearAcceleration=250, maxLinearVelocity=150, velocityScaleZ=0, worldAlpha=1}
   }
   n4 = "Constraint - skirt_rag_leg_r_A1 (P)"(ComponentPose=@n3, Alpha=1, ConstraintWeights_0=1) {
    alphaBoolBlend {blendOption=Linear}
    boneToModify {boneName=skirt_rag_leg_r_A1}
    constraintSetup [{targetBone {boneName=thigh_r}}]
    constraintWeights [1]
   }
   n5 = "Transform (Modify) Bone - Bone: attach_backpack"(ComponentPose=@n4, Alpha=1, Translation="0,3,0", Rotation="0,0,0", Scale="1,1,1") {
    translationMode=BMM_Additive
    translationSpace=BCS_BoneSpace
    alphaBoolBlend {blendOption=Linear}
    boneToModify {boneName=attach_backpack}
    translation {y=3}
   }
   Output Pose(Result=Component To Local(ComponentPose=@n5)) {layerGroup=None}
  ]}
 EventGraph {nodes [
   n1 = Event OnCharacterCustomizationLoaded()
   > Set Head(Head:Skeletal Mesh Component Object Reference=Get Array Item with EFort Custom Part Type(Array=@n1.CustomMeshParts, PartType=Head)) -> (Output_Get:Skeletal Mesh Component Object Reference)
   Event BlueprintUpdateAnimation()
  ]}
]
variables [Head {type=SkeletalMeshComponent}]
